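; PCG unit test graph for subgraph collapse, built around a few hard-coded points and two float parameters.
; Two maths nodes multiply point positions by each parameter, and a merge combines the second result with the original points via a named reroute.
; An AttributeToPosition subgraph receives FloatConstant as an override.
; A MergeAndSort subgraph gathers both products, the merge and the first subgraph's output, and supplies the sorted graph output.
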(graph UnitTestSubgraphCollapse
  (params (FloatConstant :type Double :default 5e+01) (AnotherConstant :type Double :default 5e+02))
  (bind CreatePoints_0 (Create%20Points :PointsToCreate [(struct) (struct :Transform (struct :Translation (0.0 0.0 1e+02))) (struct :Transform (struct :Translation (0.0 0.0 -1e+02)))])) ; Create 3 points
  (bind AttributeMathsOp_1 (Attribute%20Maths%20Op :InputSource1 $Position :Operation Multiply :InB FloatConstant)) ; Multiply position by constant
  (bind AttributeMathsOp_5 (Attribute%20Maths%20Op :InputSource1 $Position :Operation Multiply :InB AnotherConstant)) ; Multiply Position by other constant
  (bind Subgraph_6 (Subgraph :Graph /PCG/UnitTest/UnitTestSubgraphCollapse_AttributeToPosition.UnitTestSubgraphCollapse_AttributeToPosition :FloatConstant FloatConstant)) ; Multiply value is set as Position.Z
  (bind MergePoints_0 (Merge%20Points))
  (bind Subgraph_7 (Subgraph :Graph /PCG/UnitTest/UnitTestSubgraphCollapse_MergeAndSort.UnitTestSubgraphCollapse_MergeAndSort))
  (reroute Create%20Points%20Out CreatePoints_0)
  (connect Create%20Points%20Out Out AttributeMathsOp_1 InA)
  (connect Create%20Points%20Out Out AttributeMathsOp_5 InA)
  (connect AttributeMathsOp_5 Out MergePoints_0 In)
  (connect Create%20Points%20Out Out MergePoints_0 In)
  (connect AttributeMathsOp_1 Out Subgraph_7 In)
  (connect AttributeMathsOp_5 Out Subgraph_7 In)
  (connect MergePoints_0 Out Subgraph_7 In)
  (connect Subgraph_6 Out1 Subgraph_7 In)
  (output Out Subgraph_7.Sort%20Attributes%20Out))
