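; PCG unit-test subgraph that turns a float constant parameter into point data with a computed height, apparently to exercise subgraph collapse.
; The FloatConstant input feeds both sides of a Multiply attribute math op, squaring the value.
; The result becomes point data, and Copy Attributes, with that data as both source and target, writes the value into Position.Z for output on Out1.
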(graph UnitTestSubgraphCollapse_AttributeToPosition
  (inputs (FloatConstant :type Param))
  (outputs (Out1 :type Point))
  (bind AttributeMathsOp_0 (Attribute%20Maths%20Op :Operation Multiply :OutputTarget None))
  (bind AttributeSetToPoint_1 (Attribute%20Set%20To%20Point))
  (bind CopyAttribute_2 (Copy%20Attributes :InputSource None :OutputTarget $Position.Z))
  (connect DefaultInputNode FloatConstant AttributeMathsOp_0 InA)
  (connect DefaultInputNode FloatConstant AttributeMathsOp_0 InB)
  (connect AttributeMathsOp_0 Out AttributeSetToPoint_1 In)
  (connect AttributeSetToPoint_1 Out CopyAttribute_2 Source)
  (connect AttributeSetToPoint_1 Out CopyAttribute_2 Target)
  (output Out1 CopyAttribute_2))
