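; PCG unit-test graph for subgraph collapse that merges input points and sorts them.
; The graph input feeds a Merge Points node, then a Sort Attributes node orders the result by Position.Z with stable sort off, exposed as a named Point output.
; The sort exists because the merge is disconnected and reconnected, which suggests it keeps point order consistent for the test.
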(graph UnitTestSubgraphCollapse_MergeAndSort
  (outputs (Sort%20Attributes%20Out :type Point))
  (bind MergePoints_0 (Merge%20Points))
  (bind SortAttributes_1 (Sort%20Attributes :bUseStableSort false :SortAttributes [(struct :InputSource $Position.Z)])) ; Sort since we disconnect/reconnect merge
  (connect DefaultInputNode In MergePoints_0 In)
  (connect MergePoints_0 Out SortAttributes_1 In)
  (output Sort%20Attributes%20Out SortAttributes_1))
